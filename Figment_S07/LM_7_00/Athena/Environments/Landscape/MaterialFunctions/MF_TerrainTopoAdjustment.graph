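# Material function that adjusts a terrain base texture with a topography mask, by default T_Athena_Terrain_Topo_Mask, for the Athena landscape.
# Default UVs project the mask across the terrain from world XY, centered on the origin, scaled by TerrainSize and saturated; UVs and texture are overridable inputs.
# The mask's green channel multiplies the doubled base, making mid-gray neutral, while blue outputs Cavity_Mask and red outputs RedChannel.
# The RockMask input is unconnected.
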MaterialFunction /Figment_S07/LM_7_00/Athena/Environments/Landscape/MaterialFunctions/MF_TerrainTopoAdjustment
graph {nodes [
  FunctionInput() {bUsePreviewValueAsDefault=true, inputName=RockMask, inputType=FunctionInput_Scalar}
  n1 = FunctionInput() {inputName=BaseTexture}
  n2 = WorldPosition()
  n3 = ScalarParameter() {defaultValue=306200, group=TerrainColorMatch, parameterName=TerrainSize}
  n4 = Divide(A=@n3)
  n5 = Multiply(A=@n4) {constB=-1}
  n6 = Divide(A=Subtract(A=ComponentMask(None=@n2.XYZ) {g=true, r=true}, B=AppendVector(A=@n5, B=@n5)), B=@n3)
  n7 = TextureSample(
   UVs=FunctionInput(Preview=Saturate(None=@n6)) {bUsePreviewValueAsDefault=true, inputName=UV Coordinates, inputType=FunctionInput_Vector2}
   Tex=FunctionInput(Preview=TextureObject() {samplerType=SAMPLERTYPE_LinearColor, texture=/Figment_S07/LM_7_00/Athena/Environments/Landscape/Textures/T_Athena_Terrain_Topo_Mask}) {bUsePreviewValueAsDefault=true, inputName=Mask 2, inputType=FunctionInput_Texture2D}
  ) {
   samplerSource=SSM_Wrap_WorldGroupSettings
   samplerType=SAMPLERTYPE_LinearColor
   texture=&T_Athena_Terrain_Topo_Mask
  }
  FunctionOutput(None=Multiply(A=Multiply(A=@n1) {constB=2}, B=@n7.G))
  FunctionOutput(None=@n7.B) {outputName=Cavity_Mask}
  FunctionOutput(None=@n7.R) {outputName=RedChannel}
  LandscapeLayerCoords() {mappingScale=1525}
  Multiply(A=@n4)
  ScalarParameter() {defaultValue=50432, group=TerrainColorMatch, parameterName=TerrainSizeHalf}
 ]}
