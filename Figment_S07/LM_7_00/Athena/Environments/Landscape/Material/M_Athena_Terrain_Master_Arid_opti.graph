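# Optimized Athena landscape master material for Season 7 terrain, including an arid biome.
# Slope-blended grass, rock and arid sand form the base, with forest, mud/crater, road, gravel and wavy sand height-blended on top.
# Low quality and mobile use a biome colormap, cheaper color overrides and macro normals, with MPC switches for colormap rendering.
# Layer weights, minus rock edges and competing layers, drive landscape grass spawning for foliage.
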Material /Figment_S07/LM_7_00/Athena/Environments/Landscape/Material/M_Athena_Terrain_Master_Arid_opti
graph {
 properties {
  bFullyRough=true
  bUseMaterialAttributes=true
  bUsedWithSplineMeshes=true
  pixelDepthOffsetMode=PDOM_Legacy
  refractionMethod=RM_IndexOfRefraction
  usedShadingModels=Default Lit
 }
 nodes [
  Constant3Vector() {desc=Farm, constant=<rgba 1 0 0 0>}
  Constant3Vector() {desc=Mountain, constant=<rgba 0 1 0 0>}
  Constant()
  Constant()
  Constant() {r=2}
  Constant() {r=20000}
  Constant()
  n1 = LandscapeLayerSample() {parameterName=Grass_Topology, previewWeight=1}
  n2 = LandscapeLayerSample() {parameterName=Grass_Only, previewWeight=1}
  n3 = ScalarParameter() {defaultValue=-65, parameterName=ShiftGrassEdgeLOW}
  n4 = ScalarParameter() {defaultValue=50, parameterName=ShiftGrassEdgeHigh}
  n5 = MaterialFunctionCall!(UseVertexInterp) {materialFunction=/Figment_S07/LM_7_00/Athena/Environments/Landscape/MaterialFunctions/MF_Landscape_GrassDistanceBlend}
  n6 = ScalarParameter() {defaultValue=2, group=Grass, parameterName=GrassSpecPower_1}
  n7 = WorldPosition() {worldPositionShaderOffset=WPT_ExcludeAllShaderOffsets}
  n8 = TextureSample(UVs=Divide(A=ComponentMask(None=@n7.XYZ) {g=true, r=true}, B=ScalarParameter() {defaultValue=700, parameterName=GrassNearScale})) {
   samplerSource=SSM_Wrap_WorldGroupSettings
   samplerType=SAMPLERTYPE_Normal
   texture=/Figment_S07/LM_7_00/Packages/DS_Fortnite_Terrain_NoLOD/Textures/T_Grasslands_N
  }
  n9 = MakeMaterialAttributes!(
   BaseColor=@n5.Grass RGB
   Metallic
   Specular=LinearInterpolate(B=ScalarParameter() {defaultValue=0.15, group=Grass, parameterName=GrassSpecAmount_1}, Alpha=Power(Base=@n5.Grass A, Exp=@n6))
   Roughness
   Anisotropy
   EmissiveColor
   Opacity=@n5.Grass RGB
   OpacityMask
   Normal=@n8.RGB
   Tangent
   WorldPositionOffset
   SubsurfaceColor
   ClearCoat
   ClearCoatRoughness
   AmbientOcclusion
   Refraction
   CustomizedUVs_0
   CustomizedUVs_1
   CustomizedUVs_2
   CustomizedUVs_3
   CustomizedUVs_4
   CustomizedUVs_5
   CustomizedUVs_6
   CustomizedUVs_7
   PixelDepthOffset
   ShadingModel
   Displacement
  )
  n10 = MaterialFunctionCall() {materialFunction=/Figment_S07/LM_7_00/Athena/Environments/Landscape/MaterialFunctions/Standard/MF_Athena_Rock_01}
  n11 = MaterialFunctionCall!(height_Offset, height_GradScale, Coords=WorldPosition().XYZ) {materialFunction=/Figment_S07/LM_7_00/Athena/Environments/Landscape/MaterialFunctions/MF_Athena_WorldHeightGrad}
  n12 = Constant()
  n13 = MaterialFunctionCall!(
   In=@n10.Result
   Color=LinearInterpolate(
    A=CollectionParameter() {collection=/Game/Athena/Environments/Landscape/MPC/MPC_Landscape_S07, parameterName=Rock_Bottom_Brightness}
    B=CollectionParameter() {collection=&MPC_Landscape_S07, parameterName=Rock_Top_Brightness}
    Alpha=MaterialFunctionCall(Transition Phase=@n11.Result, Height Texture=@n12, Contrast=@n12) {materialFunction=&HeightLerp}.Alpha
   )
  ) {materialFunction=/Figment_S07/LM_7_00/Athena/Environments/Landscape/MaterialFunctions/MF_Athena_Material_BaseMultiply}
  n14 = Multiply(
   A=Multiply(
    A=MaterialFunctionCall!(Mat attributes=@n9) {materialFunction=/Engine/Functions/Engine_MaterialFunctions02/MatLayerBlend_BreakOpacity}.Result
    B=MaterialFunctionCall(Mat attributes=@n13.Result) {materialFunction=&MatLayerBlend_BreakOpacity}.Result
   )
   B=ScalarParameter() {defaultValue=0.08, group=Grass, parameterName=Transition_Phase}
  )
  n15 = Multiply(A=ComponentMask(None=VertexNormalWS()) {b=true}, B=ScalarParameter() {defaultValue=2, group=Rock, parameterName=RockMultiply_2})
  n16 = ScalarParameter() {defaultValue=5, group=Grass, parameterName=Contrast}
  n17 = MaterialFunctionCall(Transition Phase=@n14, Height Texture=@n15, Contrast=@n16) {materialFunction=&HeightLerp}
  n18 = Multiply(A=@n17.Lerp Alpha No Contrast) {desc=Rock Alpha for Grass system}
  n19 = Add(A=@n18, B=ScalarParameter() {defaultValue=-0.24, parameterName=ShiftGrassEdge_2})
  n20 = OneMinus(None=Clamp(None=LinearInterpolate(A=@n3, B=@n4, Alpha=@n19)))
  n21 = LandscapeLayerSample() {parameterName=Mud}
  n22 = Add(A=Add(A=LandscapeLayerSample() {parameterName=Gravel}, B=@n20), B=@n21)
  n23 = LandscapeLayerSample() {parameterName=Road, previewWeight=1}
  n24 = Add(A=Add(A=Add(A=@n22, B=@n23), B=LandscapeLayerSample() {parameterName=NoFoliage}), B=LandscapeLayerSample() {parameterName=Forest})
  MaterialExpressionSubtract_24 = Subtract(A=Subtract(A=Add(A=@n1, B=@n2), B=@n20), B=@n24)
  MaterialExpressionSubtract_44 = Subtract(A=LandscapeLayerSample() {parameterName=Mountain_Grass}, B=@n24)
  MaterialExpressionSubtract_30 = Subtract(A=LandscapeLayerSample() {parameterName=Farm_Grass}, B=@n24)
  MaterialExpressionSubtract_45 = Subtract(A=LandscapeLayerSample() {parameterName=Forest}, B=Add(A=@n21, B=@n23))
  MaterialExpressionLandscapeLayerSample_33 = LandscapeLayerSample() {parameterName=Crater}
  MaterialExpressionSubtract_25 = Subtract(A=Subtract(A=LandscapeLayerSample() {parameterName=Gravel}, B=@n20), B=Multiply(B=@MaterialExpressionLandscapeLayerSample_33) {constA=10})
  MaterialExpressionSubtract_43 = Subtract(A=LandscapeLayerSample() {parameterName=Flowers, previewWeight=1}, B=LandscapeLayerSample() {parameterName=NoFoliage, previewWeight=1})
  n25 = Add(A=@MaterialExpressionSubtract_24, B=@MaterialExpressionSubtract_30)
  n26 = LandscapeLayerSample() {parameterName=Arid}
  MaterialExpressionClamp_2 = Clamp(None=Add(A=Add(A=@n25, B=@MaterialExpressionSubtract_44), B=@n26))
  n27 = Subtract(A=Subtract(A=@n26, B=@n20) {desc=Subtract rock}, B=LandscapeLayerSample() {parameterName=Arid_SandWavy})
  MaterialExpressionSubtract_0 = Subtract(A=@n27, B=@n24)
  MaterialExpressionSubtract_41 = Subtract(A=LandscapeLayerSample() {parameterName=Arid_DenseGrass, previewWeight=1}, B=@n20) {desc=Subtract rock}
  LandscapeGrassOutput!(
   Grass=@MaterialExpressionSubtract_24
   Mountain=@MaterialExpressionSubtract_44
   Farm=@MaterialExpressionSubtract_30
   Forest=@MaterialExpressionSubtract_45
   Gravel=@MaterialExpressionSubtract_25
   Flowers=@MaterialExpressionSubtract_43
   Bushes=@MaterialExpressionClamp_2
   Crater=@MaterialExpressionLandscapeLayerSample_33
   Arid_Sand=@MaterialExpressionSubtract_0
   Arid_GrassDense=@MaterialExpressionSubtract_41
  ) {grassTypes [
    Grass {
     grassType=/Figment_S07/LM_7_00/Athena/Environments/Landscape/Foliage/Athena_Foliage_Grass
     input {
      expression=&self:MaterialExpressionSubtract_24
      inputName=None
      mask=0
      maskA=0
      maskB=0
      maskG=0
      maskR=0
      outputIndex=0
     }
    }
    Mountain {
     grassType=/Figment_S07/LM_7_00/Athena/Environments/Landscape/Foliage/Athena_Foliage_Mountain_Grass
     input {
      expression=&self:MaterialExpressionSubtract_44
      inputName=None
      mask=0
      maskA=0
      maskB=0
      maskG=0
      maskR=0
      outputIndex=0
     }
    }
    Farm {
     grassType=/Figment_S07/LM_7_00/Athena/Environments/Landscape/Foliage/Athena_Foliage_Farm_Grass
     input {
      expression=&self:MaterialExpressionSubtract_30
      inputName=None
      mask=0
      maskA=0
      maskB=0
      maskG=0
      maskR=0
      outputIndex=0
     }
    }
    Forest {
     grassType=/Figment_S07/LM_7_00/Athena/Environments/Landscape/Foliage/Athena_Foliage_Forest
     input {
      expression=&self:MaterialExpressionSubtract_45
      inputName=None
      mask=0
      maskA=0
      maskB=0
      maskG=0
      maskR=0
      outputIndex=0
     }
    }
    Gravel {
     grassType=/Figment_S07/LM_7_00/Athena/Environments/Landscape/Foliage/Athena_Foliage_Gravel
     input {
      expression=&self:MaterialExpressionSubtract_25
      inputName=None
      mask=0
      maskA=0
      maskB=0
      maskG=0
      maskR=0
      outputIndex=0
     }
    }
    Flowers {
     grassType=/Figment_S07/LM_7_00/Athena/Environments/Landscape/Foliage/Athena_Foliage_Flowers
     input {
      expression=&self:MaterialExpressionSubtract_43
      inputName=None
      mask=0
      maskA=0
      maskB=0
      maskG=0
      maskR=0
      outputIndex=0
     }
    }
    Bushes {
     grassType=/Figment_S07/LM_7_00/Athena/Environments/Landscape/Foliage/Athena_Foliage_Bushes
     input {
      expression=&self:MaterialExpressionClamp_2
      inputName=None
      mask=0
      maskA=0
      maskB=0
      maskG=0
      maskR=0
      outputIndex=0
     }
    }
    Crater {
     grassType=/Figment_S07/LM_7_00/Athena/Environments/Landscape/Foliage/Athena_Foliage_Crater
     input {
      expression=&self:MaterialExpressionLandscapeLayerSample_33
      inputName=None
      mask=0
      maskA=0
      maskB=0
      maskG=0
      maskR=0
      outputIndex=0
     }
    }
    Arid_Sand {
     grassType=/Figment_S07/LM_7_00/Athena/Environments/Landscape/Foliage/Athena_Foliage_Arid_Grass
     input {
      expression=&self:MaterialExpressionSubtract_0
      inputName=None
      mask=0
      maskA=0
      maskB=0
      maskG=0
      maskR=0
      outputIndex=0
     }
    }
    Arid_GrassDense {
     grassType=/Figment_S07/LM_7_00/Athena/Environments/Landscape/Foliage/Athena_Foliage_Arid_GrassDense
     input {
      expression=&self:MaterialExpressionSubtract_41
      inputName=None
      mask=0
      maskA=0
      maskB=0
      maskG=0
      maskR=0
      outputIndex=0
     }
    }
   ]}
  LandscapeLayerCoords() {mappingScale=4}
  LandscapeLayerCoords() {mappingScale=4}
  LandscapeLayerCoords() {mappingScale=5}
  LandscapeLayerSample() {parameterName=Flowers}
  MaterialExpressionConstant_6 = Constant()
  MaterialExpressionConstant3Vector_13 = Constant3Vector() {constant=<rgba 0 0 1 0>}
  n28 = LandscapeLayerBlend!(
   Layer Grass_Only=@MaterialExpressionConstant_6
   Layer Grass_Topology=@MaterialExpressionConstant_6
   Layer Arid=@MaterialExpressionConstant3Vector_13
  ) {layers [
    {
     blendType=LB_WeightBlend
     constHeightInput=0
     layerName=Grass_Only
     previewWeight=1
     constLayerInput=<0 0 0>
     heightInput {expression=null, inputName=None, mask=0, maskA=0, maskB=0, maskG=0, maskR=0, outputIndex=0}
     layerInput {
      expression=&self:MaterialExpressionConstant_6
      inputName=None
      mask=0
      maskA=0
      maskB=0
      maskG=0
      maskR=0
      outputIndex=0
     }
    }
    {
     blendType=LB_WeightBlend
     constHeightInput=0
     layerName=Grass_Topology
     previewWeight=1
     constLayerInput=<0 0 0>
     heightInput {expression=null, inputName=None, mask=0, maskA=0, maskB=0, maskG=0, maskR=0, outputIndex=0}
     layerInput {
      expression=&self:MaterialExpressionConstant_6
      inputName=None
      mask=0
      maskA=0
      maskB=0
      maskG=0
      maskR=0
      outputIndex=0
     }
    }
    {
     blendType=LB_WeightBlend
     constHeightInput=0
     layerName=Arid
     previewWeight=1
     constLayerInput=<0 0 0>
     heightInput {expression=null, inputName=None, mask=0, maskA=0, maskB=0, maskG=0, maskR=0, outputIndex=0}
     layerInput {
      expression=&self:MaterialExpressionConstant3Vector_13
      inputName=None
      mask=0
      maskA=0
      maskB=0
      maskG=0
      maskR=0
      outputIndex=0
     }
    }
   ]}
  LinearInterpolate(A=@n28)
  n29 = ScalarParameter() {defaultValue=0.9, parameterName=Lawn_Color_Min}
  n30 = ScalarParameter() {defaultValue=1.1, parameterName=Lawn_Color_Max}
  n31 = TextureSample(UVs=LandscapeLayerCoords() {mappingScale=8}) {
   samplerSource=SSM_Wrap_WorldGroupSettings
   samplerType=SAMPLERTYPE_LinearColor
   texture=/Figment_S07/LM_7_00/Athena/Environments/Landscape/Textures/T_Athena_LawnMask_01_M
  }
  LinearInterpolate(B=Multiply(B=LinearInterpolate(A=@n29, B=@n30, Alpha=@n31.R))) {constAlpha=0}
  LinearInterpolate(
   A=ScalarParameter() {defaultValue=0.85, parameterName=Lawn_Spec_Min}
   B=ScalarParameter() {defaultValue=1.5, parameterName=Lawn_Spec_Max}
   Alpha=@n31.R
  )
  MaterialExpressionLinearInterpolate_52 = LinearInterpolate(A=ScalarParameter() {defaultValue=0.8, group=Grass, parameterName=Grass_Tint}, Alpha=LandscapeLayerSample() {parameterName=NoTopo})
  n32 = ComponentMask(None=CollectionParameter() {collection=&MPC_Landscape_S07, parameterName=RockTint}) {b=true, g=true, r=true}
  n33 = MaterialFunctionCall(
   Transition Phase=Add(A=CollectionParameter() {collection=&MPC_Landscape_S07, parameterName=ColormapTransitionBias}, B=@n14)
   Height Texture=@n15
   Contrast=@n16
  ) {
   desc=Height lerp for ColorMap Texture - has ColormapTransitionBias
   materialFunction=&HeightLerp
  }
  MaterialExpressionLinearInterpolate_7 = LinearInterpolate(A=@n32, B=@MaterialExpressionLinearInterpolate_52, Alpha=@n33.Alpha)
  MaterialExpressionConstant3Vector_6 = Constant3Vector() {constant=<rgba 1 1 1 0>}
  MaterialExpressionMaterialFunctionCall_14 = MaterialFunctionCall!(Rock_Diffuse, Alpha, ColorBlend_Alpha=@n17.Alpha, ColorBlend_Alpha_Bias=@n33.Alpha, In, Grass_Diffuse) {materialFunction=/Figment_S07/LM_7_00/Athena/Environments/Landscape/MaterialFunctions/MF_FarmGrass_Colors}
  MaterialExpressionMaterialFunctionCall_17 = MaterialFunctionCall!(Rock_Diffuse, ColorBlend_Alpha=@n17.Alpha, ColorBlend_Alpha_Bias=@n33.Alpha, Grass_Diffuse) {materialFunction=/Figment_S07/LM_7_00/Athena/Environments/Landscape/MaterialFunctions/MF_MountainGrass_Colors}
  n34 = LandscapeLayerBlend!(
   Layer Grass_Only=@MaterialExpressionLinearInterpolate_52
   Layer Grass_Topology=@MaterialExpressionLinearInterpolate_7
   Layer Arid=@MaterialExpressionConstant3Vector_6
   Layer Farm_Grass=@MaterialExpressionMaterialFunctionCall_14.BlendedColors_Bias
   Layer Mountain_Grass=@MaterialExpressionMaterialFunctionCall_17.Blended_Colors_Bias
  ) {layers [
    {
     blendType=LB_WeightBlend
     constHeightInput=0
     layerName=Grass_Only
     previewWeight=1
     constLayerInput=<0 0 0>
     heightInput {expression=null, inputName=None, mask=0, maskA=0, maskB=0, maskG=0, maskR=0, outputIndex=0}
     layerInput {
      expression=&self:MaterialExpressionLinearInterpolate_52
      inputName=None
      mask=0
      maskA=0
      maskB=0
      maskG=0
      maskR=0
      outputIndex=0
     }
    }
    {
     blendType=LB_WeightBlend
     constHeightInput=0
     layerName=Grass_Topology
     previewWeight=1
     constLayerInput=<0 0 0>
     heightInput {expression=null, inputName=None, mask=0, maskA=0, maskB=0, maskG=0, maskR=0, outputIndex=0}
     layerInput {
      expression=&self:MaterialExpressionLinearInterpolate_7
      inputName=None
      mask=0
      maskA=0
      maskB=0
      maskG=0
      maskR=0
      outputIndex=0
     }
    }
    {
     blendType=LB_WeightBlend
     constHeightInput=0
     layerName=Arid
     previewWeight=1
     constLayerInput=<0 0 0>
     heightInput {expression=null, inputName=None, mask=0, maskA=0, maskB=0, maskG=0, maskR=0, outputIndex=0}
     layerInput {
      expression=&self:MaterialExpressionConstant3Vector_6
      inputName=None
      mask=0
      maskA=0
      maskB=0
      maskG=0
      maskR=0
      outputIndex=0
     }
    }
    {
     blendType=LB_WeightBlend
     constHeightInput=0
     layerName=Farm_Grass
     previewWeight=0
     constLayerInput=<0 0 0>
     heightInput {expression=null, inputName=None, mask=0, maskA=0, maskB=0, maskG=0, maskR=0, outputIndex=0}
     layerInput {
      expression=&self:MaterialExpressionMaterialFunctionCall_14
      inputName=None
      mask=0
      maskA=0
      maskB=0
      maskG=0
      maskR=0
      outputIndex=1
     }
    }
    {
     blendType=LB_WeightBlend
     constHeightInput=0
     layerName=Mountain_Grass
     previewWeight=0
     constLayerInput=<0 0 0>
     heightInput {expression=null, inputName=None, mask=0, maskA=0, maskB=0, maskG=0, maskR=0, outputIndex=0}
     layerInput {
      expression=&self:MaterialExpressionMaterialFunctionCall_17
      inputName=None
      mask=0
      maskA=0
      maskB=0
      maskG=0
      maskR=0
      outputIndex=2
     }
    }
   ]}
  LinearInterpolate(A=@n34, Alpha=@n33.Alpha)
  n35 = TextureObject() {samplerType=SAMPLERTYPE_LinearColor, texture=/Figment_S07/LM_7_00/Athena/Environments/Landscape/Textures/T_Athena_AridTest_Mask_Masks}
  n36 = MaterialFunctionCall!(Coords, Color Tint Override) {materialFunction=/Figment_S07/LM_7_00/Athena/Environments/Landscape/MaterialFunctions/Standard/MF_Athena_Mud_01}
  n37 = MaterialFunctionCall!(
   UVs=LandscapeLayerCoords() {mappingScale=6}
   UV_2ndNormal=LandscapeLayerCoords() {mappingScale=1}
   Layer=LandscapeLayerSample() {parameterName=Crater}
  ) {
   desc=Normal isnt part of main mat for now.
   materialFunction=/Figment_S07/LM_7_00/Athena/Environments/Landscape/MaterialFunctions/Standard/MF_Athena_Crater_01
  }
  n38 = Clamp(None=LinearInterpolate(
    A=ScalarParameter() {defaultValue=-0.2522, parameterName=MudCraterTransition_High}
    B=ScalarParameter() {defaultValue=10.999, parameterName=MudCraterTransition_LOW}
    Alpha=LandscapeLayerSample() {parameterName=Crater}
   ))
  n39 = MaterialFunctionCall!(Base Material=@n36.Result, Top Material=@n37.Result, Alpha=@n38) {materialFunction=/Engine/Functions/MaterialLayerFunctions/MatLayerBlend_Standard}
  n40 = MaterialFunctionCall!(Material=@n39.Blended Material) {materialFunction=/Engine/Functions/MaterialLayerFunctions/MatLayerBlend_BreakBaseColor}
  n41 = ScalarParameter() {parameterName=Arid_MudDesat}
  n42 = Add(A=ComponentMask(None=WorldPosition().XYZ) {b=true}, B=ScalarParameter() {parameterName=Sand_Grad_Offset})
  n43 = ScalarParameter() {parameterName=Sand_Grad_Range}
  n44 = LinearInterpolate(
   A=VectorParameter() {parameterName=Sand_Low_Tint}.RGB
   B=VectorParameter() {parameterName=Sand_High_Tint, defaultValue=<rgba 1 1 1 0>}.RGB
   Alpha=Clamp(None=Divide(A=@n42, B=@n43))
  )
  n45 = Multiply(
   A=Multiply(
    A=Desaturation(None=@n40.BaseColor, Fraction=@n41)
    B=VectorParameter() {parameterName=Arid_Mud_Tint, defaultValue=<rgba 2 1.0179 0.655 1>}.RGB
   )
   B=@n44
  )
  n46 = Add(A=LandscapeLayerSample() {parameterName=Arid}, B=LandscapeLayerSample() {parameterName=Arid_SandWavy})
  MaterialExpressionMaterialFunctionCall_27 = MaterialFunctionCall(In=@n9, Color=@MaterialExpressionLinearInterpolate_52) {materialFunction=&MF_Athena_Material_BaseMultiply}
  n47 = MaterialFunctionCall(Mat attributes=@MaterialExpressionMaterialFunctionCall_27.Result) {materialFunction=&MatLayerBlend_BreakOpacity}
  n48 = OneMinus(None=@n47.Result)
  n49 = LinearInterpolate(
   A=@n48
   B=MaterialFunctionCall(Mat attributes=@n36.Result) {materialFunction=&MatLayerBlend_BreakOpacity}.Result
   Alpha=LandscapeLayerSample() {parameterName=Arid}
  )
  n50 = MaterialFunctionCall!(DerivativeUVs, UVs=LandscapeLayerCoords() {mappingScale=4}) {materialFunction=/Figment_S07/LM_7_00/Athena/Environments/Landscape/MaterialFunctions/Standard/MF_Athena_Forest_01}
  n51 = ComponentMask(None=WorldPosition().XYZ) {g=true, r=true}
  n52 = ScalarParameter() {defaultValue=306000, parameterName=LandscapeSize, sortPriority=-153000}
  n53 = ScalarParameter() {defaultValue=-153000, parameterName=LandscapeMin, sortPriority=-153000}
  n54 = TextureSampleParameter2D(UVs=Subtract(A=Divide(A=@n51, B=@n52), B=Divide(A=@n53, B=@n52))) {
   parameterName=BiomeColormap
   samplerType=SAMPLERTYPE_LinearColor
   texture=/Figment_S07/LM_7_00/Athena/Environments/Landscape/Textures/T_Athena_Terrain_CombinedColors_01
  }
  n55 = MaterialFunctionCall(Base Material=@n13.Result, Top Material=@n9, Alpha=@n17.Alpha) {materialFunction=&MatLayerBlend_Standard}
  MaterialExpressionMaterialFunctionCall_24 = MaterialFunctionCall(In=@n55.Blended Material, Color=LinearInterpolate(A=@n32, B=@MaterialExpressionLinearInterpolate_52, Alpha=@n17.Alpha)) {materialFunction=&MF_Athena_Material_BaseMultiply}
  n56 = MaterialFunctionCall!(In, UseRoughness) {materialFunction=/Figment_S07/LM_7_00/Athena/Environments/Landscape/MaterialFunctions/Arid/MF_Athena_Arid_Sand}
  n57 = VertexNormalWS()
  n58 = Add(
   A=LinearInterpolate(
    A=ScalarParameter() {defaultValue=-1.5, group=Arid_SandRockBlend, parameterName=Arid_SandRock_LOW}
    B=ScalarParameter() {defaultValue=1.5, group=Arid_SandRockBlend, parameterName=Arid_SandRock_High}
    Alpha=ComponentMask(None=@n57) {b=true}
   )
   B=ScalarParameter() {defaultValue=-0.5429, group=Arid_SandRockBlend, parameterName=Arid_SandRock_Offset}
  )
  MaterialExpressionMaterialFunctionCall_127 = MaterialFunctionCall(
   Base Material=MaterialFunctionCall(
    In=MaterialFunctionCall!(AridMasks, ProjContrast) {materialFunction=/Figment_S07/LM_7_00/Athena/Environments/Landscape/MaterialFunctions/Arid/MF_Athena_Arid_Rock}.Mat
    Color=MaterialFunctionCall() {materialFunction=/Figment_S07/LM_7_00/Athena/Environments/Landscape/MaterialFunctions/MF_Athena_SedimentGradient}.Result
   ) {materialFunction=&MF_Athena_Material_BaseMultiply}.Result
   Top Material=MaterialFunctionCall(In=@n56.Simple_Material, Color=@n44) {materialFunction=&MF_Athena_Material_BaseMultiply}.Result
   Alpha=MaterialFunctionCall(
    Transition Phase=Clamp(None=@n58)
    Height Texture=OneMinus(None=@n56.Height)
    Contrast=ScalarParameter() {defaultValue=3, parameterName=Desert_SandRock_Contrast}
   ) {desc=Base Color High end, materialFunction=&HeightLerp}.Alpha
  ) {materialFunction=&MatLayerBlend_Standard}
  MaterialExpressionOneMinus_1 = OneMinus(None=@n47.Result)
  MaterialExpressionMaterialFunctionCall_18 = MaterialFunctionCall(In=@n55.Blended Material, Color=@MaterialExpressionMaterialFunctionCall_14.BlendedColors) {materialFunction=&MF_Athena_Material_BaseMultiply}
  MaterialExpressionMaterialFunctionCall_20 = MaterialFunctionCall(In=@n55.Blended Material, Color=@MaterialExpressionMaterialFunctionCall_17.Blended_Colors) {materialFunction=&MF_Athena_Material_BaseMultiply}
  n59 = LandscapeLayerBlend!(
   Layer Grass_Only=@MaterialExpressionMaterialFunctionCall_27.Result
   Layer Grass_Topology=@MaterialExpressionMaterialFunctionCall_24.Result
   Layer Arid=@MaterialExpressionMaterialFunctionCall_127.Blended Material
   Height Arid=@MaterialExpressionOneMinus_1
   Layer Farm_Grass=@MaterialExpressionMaterialFunctionCall_18.Result
   Layer Mountain_Grass=@MaterialExpressionMaterialFunctionCall_20.Result
  ) {layers [
    {
     blendType=LB_WeightBlend
     constHeightInput=0
     layerName=Grass_Only
     previewWeight=1
     constLayerInput=<0 0 0>
     heightInput {expression=null, inputName=None, mask=0, maskA=0, maskB=0, maskG=0, maskR=0, outputIndex=0}
     layerInput {
      expression=&self:MaterialExpressionMaterialFunctionCall_27
      inputName=None
      mask=0
      maskA=0
      maskB=0
      maskG=0
      maskR=0
      outputIndex=0
     }
    }
    {
     blendType=LB_WeightBlend
     constHeightInput=0
     layerName=Grass_Topology
     previewWeight=1
     constLayerInput=<0 0 0>
     heightInput {expression=null, inputName=None, mask=0, maskA=0, maskB=0, maskG=0, maskR=0, outputIndex=0}
     layerInput {
      expression=&self:MaterialExpressionMaterialFunctionCall_24
      inputName=None
      mask=0
      maskA=0
      maskB=0
      maskG=0
      maskR=0
      outputIndex=0
     }
    }
    {
     blendType=LB_HeightBlend
     constHeightInput=0
     layerName=Arid
     previewWeight=1
     constLayerInput=<0 0 0>
     heightInput {
      expression=&self:MaterialExpressionOneMinus_1
      inputName=None
      mask=0
      maskA=0
      maskB=0
      maskG=0
      maskR=0
      outputIndex=0
     }
     layerInput {
      expression=&self:MaterialExpressionMaterialFunctionCall_127
      inputName=None
      mask=0
      maskA=0
      maskB=0
      maskG=0
      maskR=0
      outputIndex=0
     }
    }
    {
     blendType=LB_WeightBlend
     constHeightInput=0
     layerName=Farm_Grass
     previewWeight=1
     constLayerInput=<0 0 0>
     heightInput {expression=null, inputName=None, mask=0, maskA=0, maskB=0, maskG=0, maskR=0, outputIndex=0}
     layerInput {
      expression=&self:MaterialExpressionMaterialFunctionCall_18
      inputName=None
      mask=0
      maskA=0
      maskB=0
      maskG=0
      maskR=0
      outputIndex=0
     }
    }
    {
     blendType=LB_WeightBlend
     constHeightInput=0
     layerName=Mountain_Grass
     previewWeight=1
     constLayerInput=<0 0 0>
     heightInput {expression=null, inputName=None, mask=0, maskA=0, maskB=0, maskG=0, maskR=0, outputIndex=0}
     layerInput {
      expression=&self:MaterialExpressionMaterialFunctionCall_20
      inputName=None
      mask=0
      maskA=0
      maskB=0
      maskG=0
      maskR=0
      outputIndex=0
     }
    }
   ]}
  n60 = MaterialFunctionCall(
   In=@n59
   Color=Multiply(A=@n54.RGB, B=CollectionParameter() {collection=&MPC_Landscape_S07, parameterName=ColormapDivisor})
  ) {materialFunction=&MF_Athena_Material_BaseMultiply}
  n61 = CollectionParameter() {collection=&MPC_Landscape_S07, parameterName=DisplayLowDetailShader}
  n62 = StaticSwitchParameter(
   True=MaterialFunctionCall!(
    Material=@n59
    NewBaseColor=MaterialFunctionCall(Material=@n60.Result) {materialFunction=&MatLayerBlend_BreakBaseColor}.BaseColor
    Mask=Saturate(None=@n61)
   ) {materialFunction=/Engine/Functions/MaterialLayerFunctions/MatLayerBlend_OverrideBaseColor}.Blended Material
   False=@n59
  ) {group=Utilities, parameterName=EnableQuickColorRender}
  n63 = QualitySwitch(Default=@n62, Low=@n60.Result, High=@n62, Medium=@n62) {desc=Use Colormap for Low/Mobile - Normal Blends for Medium/High/Epic}
  n64 = FeatureLevelSwitch(Default=@n63, ES2=@n60.Result, ES3_1=@n60.Result, SM4_REMOVED=@n63, SM5=@n63, SM6=@n63)
  n65 = MaterialFunctionCall(
   Base Material=@n64
   Top Material=@n50.Result
   Alpha=MaterialFunctionCall(
    Transition Phase=LandscapeLayerSample() {parameterName=Forest, previewWeight=1}
    Height Texture=@n50.Height
    Contrast=ScalarParameter() {defaultValue=0.25, group=Forest, parameterName=Forest_Contrast}
   ) {materialFunction=&HeightLerp}.Alpha
  ) {materialFunction=&MatLayerBlend_Standard}
  n66 = MaterialFunctionCall(
   Base Material=@n65.Blended Material
   Top Material=LandscapeLayerSwitch!(
    LayerUsed=MaterialFunctionCall(Material=@n39.Blended Material, NewBaseColor=@n45, Mask=@n46) {materialFunction=&MatLayerBlend_OverrideBaseColor}.Blended Material
    LayerNotUsed=@n39.Blended Material
   ) {parameterName=Arid}
   Alpha=MaterialFunctionCall(
    Transition Phase=LandscapeLayerSample() {parameterName=Mud}
    Height Texture=LandscapeLayerSwitch(LayerUsed=@n49, LayerNotUsed=@n48) {parameterName=Arid}
    Contrast=ScalarParameter() {defaultValue=1, group=Mud, parameterName=Mud_Contrast}
   ) {materialFunction=&HeightLerp}.Alpha
  ) {materialFunction=&MatLayerBlend_Standard}
  n67 = MaterialFunctionCall!(In=LandscapeLayerCoords() {mappingScale=4}) {materialFunction=/Figment_S07/LM_7_00/Athena/Environments/Landscape/MaterialFunctions/Standard/MF_Athena_Road_01}
  n68 = LandscapeLayerSample() {parameterName=Road, previewWeight=1}
  n69 = MaterialFunctionCall(
   Base Material=@n39.Blended Material
   Top Material=@n67.Result
   Alpha=MaterialFunctionCall(
    Transition Phase=Power(Base=@n68)
    Height Texture=MaterialFunctionCall(Mat attributes=@n67.Result) {materialFunction=&MatLayerBlend_BreakOpacity}.Result
    Contrast=ScalarParameter() {defaultValue=5, group=Road, parameterName=Road_Contrast}
   ) {materialFunction=&HeightLerp}.Alpha
  ) {materialFunction=&MatLayerBlend_Standard}
  n70 = MaterialFunctionCall(
   Transition Phase=Multiply(
    A=LandscapeLayerSample() {parameterName=Road, previewWeight=1}
    B=ScalarParameter() {defaultValue=1.2, group=Road, parameterName=Raod_Transition}
   )
   Height Texture=Multiply(A=OneMinus(None=@n47.Result), B=ScalarParameter() {defaultValue=0.25, group=Road, parameterName=Road_heightMP_2})
   Contrast=ScalarParameter() {defaultValue=5, group=Road, parameterName=RoadMud_Contrast}
  ) {materialFunction=&HeightLerp}
  n71 = MaterialFunctionCall!(Coords=LandscapeLayerCoords() {mappingScale=4}, Color Tint) {materialFunction=/Figment_S07/LM_7_00/Athena/Environments/Landscape/MaterialFunctions/Standard/MF_Athena_Gravel_01}
  n72 = VectorParameter() {parameterName=Gravel_CraterColor, defaultValue=<rgba 0.8542 0.6349 0.4402 1>}
  n73 = LandscapeLayerSwitch(
   LayerUsed=MaterialFunctionCall(In=@n71.Result, Color=LinearInterpolate(B=@n72.RGB, Alpha=@n38) {constA=1}) {materialFunction=&MF_Athena_Material_BaseMultiply}.Result
   LayerNotUsed=@n71.Result
  ) {parameterName=Crater}
  n74 = VectorParameter() {parameterName=GravelAridTint, defaultValue=<rgba 2 1.0179 0.655 1>}
  n75 = LandscapeLayerSwitch(
   LayerUsed=MaterialFunctionCall(Material=@n73, NewBaseColor=Multiply(B=@n74.RGB), Mask=LandscapeLayerSample() {parameterName=Arid}) {desc=Plug in Arid, materialFunction=&MatLayerBlend_OverrideBaseColor}.Blended Material
   LayerNotUsed=@n73
  ) {parameterName=Arid}
  n76 = LandscapeLayerSample() {parameterName=Gravel, previewWeight=1}
  n77 = TextureSample(UVs=LandscapeLayerCoords() {mappingScale=130}) {
   samplerSource=SSM_Wrap_WorldGroupSettings
   samplerType=SAMPLERTYPE_LinearColor
   texture=/Figment_S07/LM_7_00/Athena/Environments/Landscape/Textures/T_Athena_Terrain_Topo_Mask
  }
  n78 = LinearInterpolate(
   A=@n71.Height
   B=Multiply(A=@n77.R, B=ScalarParameter() {defaultValue=0.5, parameterName=HeightDistanceBlend})
   Alpha=MaterialFunctionCall!(Distance) {materialFunction=/Figment_S07/LM_7_00/Athena/Environments/Landscape/MaterialFunctions/MF_TerrainDistanceFade}.Result
  ) {constB=0.5}
  n79 = FeatureLevelSwitch(Default=@n78, ES2=@n71.Height, ES3_1=@n71.Height, SM4_REMOVED=@n78, SM5=@n78, SM6=@n78)
  n80 = ScalarParameter() {group=Gravel, parameterName=Gravel_BlendContrast}
  n81 = MaterialFunctionCall!(In=LandscapeLayerCoords() {mappingScale=4}) {materialFunction=/Figment_S07/LM_7_00/Athena/Environments/Landscape/MaterialFunctions/Arid/MF_Athena_Arid_SandWavy}
  n82 = MaterialFunctionCall(
   Transition Phase=LandscapeLayerSample() {parameterName=Arid_SandWavy}
   Height Texture=@n81.Height
   Contrast=ScalarParameter() {defaultValue=1, parameterName=Arid_Sandwavy_Contrast}
  ) {materialFunction=&HeightLerp}
  n83 = MaterialFunctionCall(
   Base Material=MaterialFunctionCall(
    Base Material=MaterialFunctionCall(Base Material=@n66.Blended Material, Top Material=@n69.Blended Material, Alpha=@n70.Alpha) {materialFunction=&MatLayerBlend_Standard}.Blended Material
    Top Material=@n75
    Alpha=MaterialFunctionCall(Transition Phase=@n76, Height Texture=@n79, Contrast=@n80) {materialFunction=&HeightLerp}.Alpha
   ) {materialFunction=&MatLayerBlend_Standard}.Blended Material
   Top Material=MaterialFunctionCall(In=@n81.Result, Color=@n44) {materialFunction=&MF_Athena_Material_BaseMultiply}.Result
   Alpha=@n82.Alpha
  ) {materialFunction=&MatLayerBlend_Standard}
  n84 = MaterialFunctionCall(Material=@n83.Blended Material) {materialFunction=&MatLayerBlend_BreakBaseColor}
  n85 = MaterialFunctionCall!(Mask 2=@n35, UV Coordinates=LandscapeLayerCoords() {mappingScale=631}, RockMask, BaseTexture=@n84.BaseColor) {materialFunction=/Figment_S07/LM_7_00/Athena/Environments/Landscape/MaterialFunctions/MF_TerrainTopoAdjustment}
  n86 = Saturate(None=CollectionParameter() {collection=&MPC_Landscape_S07, parameterName=DisplayLowDetailShader})
  n87 = LinearInterpolate(
   A=StaticSwitchParameter(True=LinearInterpolate(A=@n85.Result, B=@n84.BaseColor, Alpha=@n86), False=@n85.Result) {group=Utilities, parameterName=EnableQuickColorRender}
   B=@n84.BaseColor
   Alpha=LandscapeLayerSample() {parameterName=NoTopo}
  )
  LinearInterpolate(A=@n87, B=@n87, Alpha=LandscapeLayerSample() {parameterName=NoFoliage, previewWeight=1})
  LinearInterpolate(A=@n87, B=@n87, Alpha=LandscapeLayerSample() {parameterName=Flowers, previewWeight=1})
  n88 = TextureObject() {samplerType=SAMPLERTYPE_LinearColor, texture=&T_Athena_AridTest_Mask_Masks}
  n89 = LandscapeLayerCoords() {mappingScale=631}
  n90 = OneMinus(None=@n17.Lerp Alpha No Contrast)
  n91 = ComponentMask(None=VertexNormalWS()) {b=true}
  n92 = ComponentMask(None=MaterialFunctionCall(In=Abs(None=@n91), Contrast=@n10.Proj_Contrast) {materialFunction=&CheapContrast}.Result) {r=true}
  n93 = Desaturation(None=@n34, Fraction=LinearInterpolate(B=ScalarParameter() {defaultValue=1, parameterName=RockEdgeDesat}, Alpha=Min(A=@n90, B=@n92))) {desc=To Render Colormap}
  n94 = Divide(
   A=LinearInterpolate(
    A=MaterialFunctionCall(Mask 2=@n88, UV Coordinates=@n89, BaseTexture=@n93) {materialFunction=&MF_TerrainTopoAdjustment}.Result
    B=@n93
    Alpha=LandscapeLayerSample() {parameterName=NoTopo}
   )
   B=CollectionParameter() {collection=&MPC_Landscape_S07, parameterName=ColormapDivisor}
  )
  n95 = SphereMask(
   A=CollectionParameter() {collection=&MPC_Landscape_S07, parameterName=GrassEdgeZ}
   B=ComponentMask(None=VertexNormalWS()) {b=true}
   Radius=CollectionParameter() {collection=&MPC_Landscape_S07, parameterName=GrassEdgeDesatWidth}
  ) {hardnessPercent=0}
  n96 = LinearInterpolate(
   A=@n87
   B=MaterialFunctionCall!(UVW=Desaturation(None=@n94, Fraction=@n95)) {materialFunction=/Engine/ArtTools/RenderToTexture/MaterialFunctions/MarkRedOutside0-1Range}.Result
   Alpha=Saturate(None=CollectionParameter() {collection=&MPC_Landscape_S07, parameterName=RenderColorMap})
  )
  n97 = Saturate(None=CollectionParameter() {collection=&MPC_Landscape_S07, parameterName=DisplayBiomeRGB})
  n98 = MaterialFunctionCall!(
   Mat_In=@n83.Blended Material
   BaseColor=StaticSwitchParameter(True=LinearInterpolate(A=@n96, B=@n28, Alpha=@n97), False=@n87) {group=Utilities, parameterName=EnableQuickColorRender}
  ) {materialFunction=/Figment_S07/LM_7_00/Athena/Environments/Landscape/MaterialFunctions/MF_Athena_ReplaceBaseColor}
  n99 = MaterialFunctionCall(
   Material=MaterialFunctionCall(
    Material=@n64
    NewBaseColor=MaterialFunctionCall(Material=@n50.Result) {materialFunction=&MatLayerBlend_BreakBaseColor}.BaseColor
    Mask=LandscapeLayerSample() {parameterName=Forest, previewWeight=1}
   ) {materialFunction=&MatLayerBlend_OverrideBaseColor}.Blended Material
   NewBaseColor=MaterialFunctionCall(Material=@n39.Blended Material) {materialFunction=&MatLayerBlend_BreakBaseColor}.BaseColor
   Mask=LandscapeLayerSample() {parameterName=Mud}
  ) {materialFunction=&MatLayerBlend_OverrideBaseColor}
  n100 = MaterialFunctionCall(Material=@n67.Result) {desc=Skip mud blend for Low/Mobile, materialFunction=&MatLayerBlend_BreakBaseColor}
  n101 = MaterialFunctionCall(
   Material=MaterialFunctionCall(
    Material=MaterialFunctionCall(Material=@n99.Blended Material, NewBaseColor=@n100.BaseColor, Mask=@n70.Alpha) {materialFunction=&MatLayerBlend_OverrideBaseColor}.Blended Material
    NewBaseColor=MaterialFunctionCall(Material=@n75) {materialFunction=&MatLayerBlend_BreakBaseColor}.BaseColor
    Mask=LandscapeLayerSample() {parameterName=Gravel, previewWeight=1}
   ) {materialFunction=&MatLayerBlend_OverrideBaseColor}.Blended Material
   NewBaseColor=MaterialFunctionCall(Material=@n81.Result) {materialFunction=&MatLayerBlend_BreakBaseColor}.BaseColor
   Mask=LandscapeLayerSample() {parameterName=Arid_SandWavy}
  ) {materialFunction=&MatLayerBlend_OverrideBaseColor}
  n102 = QualitySwitch(Default=@n98.Result, Low=@n101.Blended Material, High=@n98.Result, Medium=@n98.Result)
  n103 = BreakMaterialAttributes!(Attr=FeatureLevelSwitch(Default=@n102, ES2=@n101.Blended Material, ES3_1=@n101.Blended Material, SM4_REMOVED=@n102, SM5=@n102, SM6=@n102))
  n104 = ComponentMask(None=@n103.BaseColor) {r=true}
  n105 = LinearInterpolate(
   A=@n104
   B=Multiply(A=@n104, B=ScalarParameter() {defaultValue=0.3, parameterName=Arid_SpecMP})
   Alpha=LandscapeLayerSample() {parameterName=Arid}
  )
  n106 = Multiply(A=@n104, B=ScalarParameter() {defaultValue=1, parameterName=SimpleMobileSpec})
  n107 = MaterialFunctionCall!(Material=@n59) {materialFunction=/Engine/Functions/MaterialLayerFunctions/MatLayerBlend_BreakNormal}
  n108 = LandscapeLayerSample() {parameterName=Road, previewWeight=1}
  n109 = LandscapeLayerSample() {parameterName=Gravel, previewWeight=1}
  n110 = LinearInterpolate(
   A=LandscapeLayerSwitch(LayerUsed=LinearInterpolate(A=@n107.Normal, B=@n37.Normal, Alpha=@n38), LayerNotUsed=@n107.Normal) {parameterName=Crater}
   B=Constant3Vector() {constant=<rgba 0 0 1 0>}
   Alpha=Add(A=Add(A=@n108, B=@n109), B=LandscapeLayerSample() {parameterName=Mud, previewWeight=1})
  )
  n111 = LinearInterpolate(A=@n110, B=@n81.Normal, Alpha=@n82.Alpha)
  n112 = Divide(
   A=ComponentMask(None=WorldPosition() {worldPositionShaderOffset=WPT_ExcludeAllShaderOffsets}.XYZ) {g=true, r=true}
   B=ScalarParameter() {defaultValue=25000, parameterName=MacroNormalScale}
  )
  n113 = ScalarParameter() {defaultValue=0.625, parameterName=MacroNormalIntensity}
  n114 = Constant() {r=0.25}
  n115 = Constant() {r=0.9}
  n116 = FeatureLevelSwitch(Default=@n114, ES2=@n115, ES3_1=@n115, SM4_REMOVED=@n114)
  n117 = Multiply(
   A=LinearInterpolate(
    A=Constant3Vector() {constant=<rgba 0 0 1 0>}
    B=TextureSample(UVs=@n112) {samplerType=SAMPLERTYPE_Normal, texture=/Game/ContentCreationTools/TextureCreation/Textures/T_BPCreated_MacroNormal_01}.RGB
    Alpha=MaterialFunctionCall() {materialFunction=&MF_TerrainDistanceFade}.Result
   )
   B=MaterialFunctionCall(X=@n113, Y=@n113, Z=QualitySwitch(Default=@n116, Low=@n115)) {materialFunction=&MakeFloat3}.Result
  )
  n118 = Add(
   A=Multiply(A=LandscapeLayerSwitch(LayerUsed=@n111, LayerNotUsed=@n110) {parameterName=Arid}, B=Constant3Vector() {constant=<rgba 1 1 0.5 0>})
   B=@n117
  )
  n119 = QualitySwitch(Default=@n118, Low=@n117, High=@n118, Medium=@n118)
  MakeMaterialAttributes(
   BaseColor=@n103.BaseColor
   Specular=StaticSwitchParameter(True=FeatureLevelSwitch(Default=@n105, ES2=@n106, ES3_1=@n106, SM4_REMOVED=@n105, SM5=@n105, SM6=@n105), False=@n103.Specular) {defaultValue=true, parameterName=UseCheapSpec}
   Normal=FeatureLevelSwitch(Default=@n119, ES2=@n117, ES3_1=@n117, SM4_REMOVED=@n119, SM5=@n118, SM6=@n118)
  )
  MaterialFunctionCall(Material=@n73) {materialFunction=&MatLayerBlend_BreakBaseColor}
  MaterialFunctionCall(Mat attributes=@n65.Blended Material) {materialFunction=&MatLayerBlend_BreakOpacity}
  MaterialFunctionCall(Material=@n59) {materialFunction=&MatLayerBlend_BreakBaseColor}
  MaterialFunctionCall() {materialFunction=&MF_Landscape_GrassDistanceBlend}
  Multiply()
  Multiply()
  Multiply(A=@n93) {desc=BiomeColors}
  n120 = Multiply() {desc=Swamp_Grass}
  n121 = ScalarParameter() {defaultValue=5, parameterName=FoliagePOW}
  Power(Base=@n120, Exp=@n121)
  n122 = Add(
   A=LandscapeLayerSample() {parameterName=Grass_Topology, previewWeight=1}
   B=LandscapeLayerSample() {parameterName=Grass_Only, previewWeight=1}
  )
  Power(Base=@n122, Exp=@n121)
  Reroute(None=@n20)
  n123 = Add(
   A=Add(A=OneMinus(None=@n18), B=LandscapeLayerSample() {parameterName=Gravel})
   B=Add(A=LandscapeLayerSample() {parameterName=Mud}, B=LandscapeLayerSample() {parameterName=Road})
  )
  Reroute(None=OneMinus(None=Saturate(None=@n123)))
  ScalarParameter() {defaultValue=1, parameterName=TestingBiomeRender}
  ScalarParameter() {defaultValue=0.85, parameterName=Roughness}
  ScalarParameter() {defaultValue=306200, parameterName=TerrainSize}
  ScalarParameter() {defaultValue=30000, parameterName=Testing_gradient}
  StaticBool() {value=true}
  n124 = LandscapeLayerSample() {parameterName=Gravel}
  n125 = Subtract(A=Subtract(A=@n122, B=@n20), B=@n124)
  n126 = LandscapeLayerSample() {parameterName=Forest}
  n127 = LandscapeLayerSample() {parameterName=Mud}
  n128 = ScalarParameter() {defaultValue=-1, parameterName=ShiftGrassEdgeRoadLOW}
  n129 = ScalarParameter() {defaultValue=10, parameterName=ShiftGrassEdgeRoadHigh}
  n130 = LandscapeLayerSample() {parameterName=Road, previewWeight=1}
  Subtract(A=Subtract(A=Subtract(A=@n125, B=@n126), B=@n127), B=Clamp(None=LinearInterpolate(A=@n128, B=@n129, Alpha=@n130)))
  Subtract(A=@n27, B=LandscapeLayerSample() {parameterName=Gravel})
  n131 = Subtract(A=Subtract(A=@n120, B=@n20), B=@n124)
  Subtract(A=Subtract(A=Subtract(A=@n131, B=@n126), B=@n127), B=@n130)
  Subtract(A=@n76, B=OneMinus(None=@n17.Alpha)) {desc=Remove Rock from gravel}
  n132 = Subtract(A=Subtract(A=Multiply() {desc=Farm_Grass}, B=@n20), B=@n124)
  Subtract(A=Subtract(A=Subtract(A=@n132, B=@n126), B=@n127), B=@n130)
  n133 = Multiply() {desc=Mountain Grass}
  n134 = Subtract(A=Subtract(A=@n133, B=@n20), B=@n124)
  Subtract(A=Subtract(A=Subtract(A=@n134, B=@n126), B=@n127), B=@n130)
  n135 = Subtract(A=LandscapeLayerSample() {parameterName=Forest}, B=@n20)
  Subtract(A=Subtract(A=@n135, B=@n127), B=@n130)
  Subtract(A=@n135, B=@n133)
 ]
}
