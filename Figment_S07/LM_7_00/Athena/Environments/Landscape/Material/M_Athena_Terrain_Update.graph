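# Landscape material for Athena terrain, blending seven painted layers from Grass_Only and Arid to Ice and Snow with height and weight blending.
# Grass_Topology slope-blends grass over rock, whose brightness follows a world-height gradient between MPC_Landscape_S07 values.
# The layer result is multiplied by biome colormap samples mixed by snow mask and slope alpha.
# Display_Rock_Ground and Display_Ground_Colors static switches select per-layer collection tints multiplied by a topology mask.
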Material /Figment_S07/LM_7_00/Athena/Environments/Landscape/Material/M_Athena_Terrain_Update
graph {
 properties {pixelDepthOffsetMode=PDOM_Legacy, usedShadingModels=Default Lit}
 nodes [
  MaterialExpressionMaterialFunctionCall_22 = MaterialFunctionCall() {materialFunction=/Figment_S07/LM_7_00/Athena/Environments/Landscape/MaterialFunctions/Standard/MF_Athena_Grass_01}
  n1 = MaterialFunctionCall!(Height=@MaterialExpressionMaterialFunctionCall_22.Height, Transition Multiplier, Height Multiplier) {materialFunction=/Figment_S07/LM_7_00/Athena/Environments/Landscape/MaterialFunctions/Utilities/MF_Athena_SlopeBlend}
  LinearInterpolate(Alpha=@n1.Alpha)
  LinearInterpolate()
  MaterialFunctionCall() {materialFunction=/Figment_S07/LM_7_00/Athena/Environments/Landscape/MaterialFunctions/Arid/MF_Athena_AridBiome}
  n2 = CollectionParameter() {collection=/Game/Athena/Environments/Landscape/MPC/MPC_Landscape_S07, parameterName=Rock_Bottom_Brightness}
  n3 = CollectionParameter() {collection=&MPC_Landscape_S07, parameterName=Rock_Top_Brightness}
  n4 = MaterialFunctionCall!(height_Offset, height_GradScale, Coords=WorldPosition().XYZ) {materialFunction=/Figment_S07/LM_7_00/Athena/Environments/Landscape/MaterialFunctions/MF_Athena_WorldHeightGrad}
  MaterialExpressionMaterialFunctionCall_31 = MaterialFunctionCall!(
   Base Material=MaterialFunctionCall!(
    In=MaterialFunctionCall() {materialFunction=/Figment_S07/LM_7_00/Athena/Environments/Landscape/MaterialFunctions/Standard/MF_Athena_Rock_01}.Result
    Color=LinearInterpolate(A=@n2, B=@n3, Alpha=@n4.Result)
   ) {materialFunction=/Figment_S07/LM_7_00/Athena/Environments/Landscape/MaterialFunctions/MF_Athena_Material_BaseMultiply}.Result
   Top Material=@MaterialExpressionMaterialFunctionCall_22.Result
   Alpha=@n1.Alpha
  ) {materialFunction=/Engine/Functions/MaterialLayerFunctions/MatLayerBlend_Standard}
  n5 = LandscapeLayerBlend!(
   Layer Grass_Only=@MaterialExpressionMaterialFunctionCall_22.Result
   Height Grass_Only=@MaterialExpressionMaterialFunctionCall_22.Result
   Layer Grass_Topology=@MaterialExpressionMaterialFunctionCall_31.Blended Material
   Layer Arid
   Layer Farm_Grass
   Layer Mountain_Grass
   Height Mountain_Grass
   Layer Ice
   Layer Snow
   Height Snow
  ) {layers [
    {
     blendType=LB_HeightBlend
     constHeightInput=0
     layerName=Grass_Only
     previewWeight=1
     constLayerInput=<0 0 0>
     heightInput {
      expression=&self:MaterialExpressionMaterialFunctionCall_22
      inputName=None
      mask=0
      maskA=0
      maskB=0
      maskG=0
      maskR=0
      outputIndex=1
     }
     layerInput {
      expression=&self:MaterialExpressionMaterialFunctionCall_22
      inputName=None
      mask=0
      maskA=0
      maskB=0
      maskG=0
      maskR=0
      outputIndex=2
     }
    }
    {
     blendType=LB_WeightBlend
     constHeightInput=0
     layerName=Grass_Topology
     previewWeight=1
     constLayerInput=<0 0 0>
     heightInput {expression=null, inputName=None, mask=0, maskA=0, maskB=0, maskG=0, maskR=0, outputIndex=0}
     layerInput {
      expression=&self:MaterialExpressionMaterialFunctionCall_31
      inputName=None
      mask=0
      maskA=0
      maskB=0
      maskG=0
      maskR=0
      outputIndex=0
     }
    }
    {
     blendType=LB_WeightBlend
     constHeightInput=0
     layerName=Arid
     previewWeight=1
     constLayerInput=<0 0 0>
     heightInput {expression=null, inputName=None, mask=0, maskA=0, maskB=0, maskG=0, maskR=0, outputIndex=0}
     layerInput {expression=null, inputName=None, mask=0, maskA=0, maskB=0, maskG=0, maskR=0, outputIndex=0}
    }
    {
     blendType=LB_WeightBlend
     constHeightInput=0
     layerName=Farm_Grass
     previewWeight=1
     constLayerInput=<0 0 0>
     heightInput {expression=null, inputName=None, mask=0, maskA=0, maskB=0, maskG=0, maskR=0, outputIndex=0}
     layerInput {expression=null, inputName=None, mask=0, maskA=0, maskB=0, maskG=0, maskR=0, outputIndex=3}
    }
    {
     blendType=LB_HeightBlend
     constHeightInput=0
     layerName=Mountain_Grass
     previewWeight=1
     constLayerInput=<0 0 0>
     heightInput {expression=null, inputName=None, mask=0, maskA=0, maskB=0, maskG=0, maskR=0, outputIndex=0}
     layerInput {expression=null, inputName=None, mask=0, maskA=0, maskB=0, maskG=0, maskR=0, outputIndex=2}
    }
    {
     blendType=LB_WeightBlend
     constHeightInput=0
     layerName=Ice
     previewWeight=0
     constLayerInput=<0 0 0>
     heightInput {expression=null, inputName=None, mask=0, maskA=0, maskB=0, maskG=0, maskR=0, outputIndex=0}
     layerInput {expression=null, inputName=None, mask=0, maskA=0, maskB=0, maskG=0, maskR=0, outputIndex=3}
    }
    {
     blendType=LB_HeightBlend
     constHeightInput=0
     layerName=Snow
     previewWeight=0
     constLayerInput=<0 0 0>
     heightInput {expression=null, inputName=None, mask=0, maskA=0, maskB=0, maskG=0, maskR=0, outputIndex=0}
     layerInput {expression=null, inputName=None, mask=0, maskA=0, maskB=0, maskG=0, maskR=0, outputIndex=0}
    }
   ]}
  n6 = TextureSampleParameter2D() {
   parameterName=BiomeColormap
   samplerType=SAMPLERTYPE_LinearColor
   texture=/Figment_S07/LM_7_00/Athena/Environments/Landscape/Textures/T_Athena_Terrain_CombinedColors_01
  }
  n7 = Multiply() {desc=Snow Mask}
  n8 = TextureSampleParameter2D() {
   parameterName=BiomeColormap
   samplerType=SAMPLERTYPE_LinearColor
   texture=&T_Athena_Terrain_CombinedColors_01
  }
  Multiply(A=@n5, B=LinearInterpolate(A=LinearInterpolate(A=@n6.RGB, Alpha=@n7), B=@n8.RGB, Alpha=@n1.Alpha))
  MaterialExpressionConstant_4 = Constant() {r=1}
  MaterialExpressionComponentMask_11 = ComponentMask(None=CollectionParameter() {collection=&MPC_Landscape_S07, parameterName=FarmRockTint}) {b=true, g=true, r=true}
  MaterialExpressionComponentMask_12 = ComponentMask(None=CollectionParameter() {collection=&MPC_Landscape_S07, parameterName=MountainRockTint}) {b=true, g=true, r=true}
  n9 = LandscapeLayerBlend!(
   Layer Grass_Only=@MaterialExpressionConstant_4
   Layer Grass_Topology=@MaterialExpressionConstant_4
   Layer Arid=@MaterialExpressionConstant_4
   Layer Farm_Grass=@MaterialExpressionComponentMask_11
   Layer Mountain_Grass=@MaterialExpressionComponentMask_12
   Layer Ice=@MaterialExpressionConstant_4
   Layer Snow=@MaterialExpressionConstant_4
  ) {layers [
    {
     blendType=LB_WeightBlend
     constHeightInput=0
     layerName=Grass_Only
     previewWeight=1
     constLayerInput=<0 0 0>
     heightInput {expression=null, inputName=None, mask=0, maskA=0, maskB=0, maskG=0, maskR=0, outputIndex=0}
     layerInput {
      expression=&self:MaterialExpressionConstant_4
      inputName=None
      mask=0
      maskA=0
      maskB=0
      maskG=0
      maskR=0
      outputIndex=0
     }
    }
    {
     blendType=LB_WeightBlend
     constHeightInput=0
     layerName=Grass_Topology
     previewWeight=1
     constLayerInput=<0 0 0>
     heightInput {expression=null, inputName=None, mask=0, maskA=0, maskB=0, maskG=0, maskR=0, outputIndex=0}
     layerInput {
      expression=&self:MaterialExpressionConstant_4
      inputName=None
      mask=0
      maskA=0
      maskB=0
      maskG=0
      maskR=0
      outputIndex=0
     }
    }
    {
     blendType=LB_WeightBlend
     constHeightInput=0
     layerName=Arid
     previewWeight=1
     constLayerInput=<0 0 0>
     heightInput {expression=null, inputName=None, mask=0, maskA=0, maskB=0, maskG=0, maskR=0, outputIndex=0}
     layerInput {
      expression=&self:MaterialExpressionConstant_4
      inputName=None
      mask=0
      maskA=0
      maskB=0
      maskG=0
      maskR=0
      outputIndex=0
     }
    }
    {
     blendType=LB_WeightBlend
     constHeightInput=0
     layerName=Farm_Grass
     previewWeight=1
     constLayerInput=<0 0 0>
     heightInput {expression=null, inputName=None, mask=0, maskA=0, maskB=0, maskG=0, maskR=0, outputIndex=0}
     layerInput {
      expression=&self:MaterialExpressionComponentMask_11
      inputName=None
      mask=0
      maskA=0
      maskB=0
      maskG=0
      maskR=0
      outputIndex=0
     }
    }
    {
     blendType=LB_WeightBlend
     constHeightInput=0
     layerName=Mountain_Grass
     previewWeight=1
     constLayerInput=<0 0 0>
     heightInput {expression=null, inputName=None, mask=0, maskA=0, maskB=0, maskG=0, maskR=0, outputIndex=0}
     layerInput {
      expression=&self:MaterialExpressionComponentMask_12
      inputName=None
      mask=0
      maskA=0
      maskB=0
      maskG=0
      maskR=0
      outputIndex=0
     }
    }
    {
     blendType=LB_WeightBlend
     constHeightInput=0
     layerName=Ice
     previewWeight=0
     constLayerInput=<0 0 0>
     heightInput {expression=null, inputName=None, mask=0, maskA=0, maskB=0, maskG=0, maskR=0, outputIndex=0}
     layerInput {
      expression=&self:MaterialExpressionConstant_4
      inputName=None
      mask=0
      maskA=0
      maskB=0
      maskG=0
      maskR=0
      outputIndex=0
     }
    }
    {
     blendType=LB_WeightBlend
     constHeightInput=0
     layerName=Snow
     previewWeight=0
     constLayerInput=<0 0 0>
     heightInput {expression=null, inputName=None, mask=0, maskA=0, maskB=0, maskG=0, maskR=0, outputIndex=0}
     layerInput {
      expression=&self:MaterialExpressionConstant_4
      inputName=None
      mask=0
      maskA=0
      maskB=0
      maskG=0
      maskR=0
      outputIndex=0
     }
    }
   ]}
  n10 = TextureSample() {
   samplerSource=SSM_Wrap_WorldGroupSettings
   samplerType=SAMPLERTYPE_LinearColor
   texture=/Figment_S07/LM_7_00/Athena/Environments/Landscape/Textures/T_Athena_Terrain_Topo_Mask
  }
  MaterialExpressionConstant_3 = Constant() {r=1}
  MaterialExpressionComponentMask_10 = ComponentMask(None=CollectionParameter() {collection=&MPC_Landscape_S07, parameterName=FarmGrassTint}) {b=true, g=true, r=true}
  MaterialExpressionComponentMask_3 = ComponentMask(None=CollectionParameter() {collection=&MPC_Landscape_S07, parameterName=MountainGrassTint}) {b=true, g=true, r=true}
  n11 = LandscapeLayerBlend(
   Layer Grass_Only=@MaterialExpressionConstant_3
   Layer Grass_Topology=@MaterialExpressionConstant_3
   Layer Farm_Grass=@MaterialExpressionComponentMask_10
   Layer Mountain_Grass=@MaterialExpressionComponentMask_3
   Layer Ice=@MaterialExpressionConstant_3
   Layer Snow=@MaterialExpressionConstant_3
  ) {layers [
    {
     blendType=LB_WeightBlend
     constHeightInput=0
     layerName=Grass_Only
     previewWeight=1
     constLayerInput=<0 0 0>
     heightInput {expression=null, inputName=None, mask=0, maskA=0, maskB=0, maskG=0, maskR=0, outputIndex=0}
     layerInput {
      expression=&self:MaterialExpressionConstant_3
      inputName=None
      mask=0
      maskA=0
      maskB=0
      maskG=0
      maskR=0
      outputIndex=0
     }
    }
    {
     blendType=LB_WeightBlend
     constHeightInput=0
     layerName=Grass_Topology
     previewWeight=1
     constLayerInput=<0 0 0>
     heightInput {expression=null, inputName=None, mask=0, maskA=0, maskB=0, maskG=0, maskR=0, outputIndex=0}
     layerInput {
      expression=&self:MaterialExpressionConstant_3
      inputName=None
      mask=0
      maskA=0
      maskB=0
      maskG=0
      maskR=0
      outputIndex=0
     }
    }
    {
     blendType=LB_WeightBlend
     constHeightInput=0
     layerName=Arid
     previewWeight=1
     constLayerInput=<0 0 0>
     heightInput {expression=null, inputName=None, mask=0, maskA=0, maskB=0, maskG=0, maskR=0, outputIndex=0}
     layerInput {expression=null, inputName=None, mask=0, maskA=0, maskB=0, maskG=0, maskR=0, outputIndex=1}
    }
    {
     blendType=LB_WeightBlend
     constHeightInput=0
     layerName=Farm_Grass
     previewWeight=1
     constLayerInput=<0 0 0>
     heightInput {expression=null, inputName=None, mask=0, maskA=0, maskB=0, maskG=0, maskR=0, outputIndex=0}
     layerInput {
      expression=&self:MaterialExpressionComponentMask_10
      inputName=None
      mask=0
      maskA=0
      maskB=0
      maskG=0
      maskR=0
      outputIndex=0
     }
    }
    {
     blendType=LB_WeightBlend
     constHeightInput=0
     layerName=Mountain_Grass
     previewWeight=1
     constLayerInput=<0 0 0>
     heightInput {expression=null, inputName=None, mask=0, maskA=0, maskB=0, maskG=0, maskR=0, outputIndex=0}
     layerInput {
      expression=&self:MaterialExpressionComponentMask_3
      inputName=None
      mask=0
      maskA=0
      maskB=0
      maskG=0
      maskR=0
      outputIndex=0
     }
    }
    {
     blendType=LB_WeightBlend
     constHeightInput=0
     layerName=Ice
     previewWeight=0
     constLayerInput=<0 0 0>
     heightInput {expression=null, inputName=None, mask=0, maskA=0, maskB=0, maskG=0, maskR=0, outputIndex=0}
     layerInput {
      expression=&self:MaterialExpressionConstant_3
      inputName=None
      mask=0
      maskA=0
      maskB=0
      maskG=0
      maskR=0
      outputIndex=0
     }
    }
    {
     blendType=LB_WeightBlend
     constHeightInput=0
     layerName=Snow
     previewWeight=0
     constLayerInput=<0 0 0>
     heightInput {expression=null, inputName=None, mask=0, maskA=0, maskB=0, maskG=0, maskR=0, outputIndex=0}
     layerInput {
      expression=&self:MaterialExpressionConstant_3
      inputName=None
      mask=0
      maskA=0
      maskB=0
      maskG=0
      maskR=0
      outputIndex=0
     }
    }
   ]}
  n12 = Multiply(A=@n11) {constB=2}
  n13 = TextureSample() {
   samplerSource=SSM_Wrap_WorldGroupSettings
   samplerType=SAMPLERTYPE_LinearColor
   texture=&T_Athena_Terrain_Topo_Mask
  }
  StaticSwitchParameter(
   True=Multiply(A=Multiply(A=@n9) {constB=2}, B=@n10.G)
   False=StaticSwitchParameter(True=Multiply(A=@n12, B=@n13.G)) {parameterName=Display_Ground_Colors}
  ) {parameterName=Display_Rock_Ground}
 ]
}
